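; Parent item spawner actor for Athena that spawns loot from a configurable loot tier group at its own location.
; On the server, once playlist data is ready, a hotfixable row toggle and spawn-chance roll decide between spawning and removing the spawner.
; Spawned pickups are turned to match the spawner's forward direction.
; Removal, also optional after spawning, uses a short randomized lifespan rather than immediate destruction.
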(event EventBeginPlay
  ;[1 1 1] Begin Play - On the server when the playlist is ready spawn item if hotfixed on. Else kill the spawner
  (if (Networking|IsServer)
    (Class|FortAsyncActionCurrentPlaylistReady|PlaylistDataReadyAsync
      (:OnReady
        (if (and (Class|AthenaFunctionLibrary|RowtoBool (Variables|Default|GetRow_Spawn)) (CallFunction|GetChancetoSpawn))
          (CallFunction|SpawnItem)
          (else
            (CallFunction|KillSpawner)))))))

(event Custom|SpawnItem
  ;[1 1 1] Spawn Item - Spawn item and save out reference
  (bind _itempickupsspawned (Class|AthenaFunctionLibrary|SpawnLootfromTable (Variables|Default|GetLootTierGroupToSpawn) 1 -1 (Transformation|GetActorLocation)))
  (for _array_element _itempickupsspawned
    ; Set Rotation after spawning actor
    (Transformation|SetActorRotation _array_element (Math|Conversions|RotationFromXVector (Math|Rotator|GetRotationXVector (Transformation|GetActorRotation))) false))
  ;[1 1 1] Set lifespawn after spawn if desired
  (if (Variables|Default|GetDestroyAfterSpawn)
    (CallFunction|KillSpawner)))

(event Custom|KillSpawner
  ;[1 1 1] Kill Spawner - Set the lifespan
  (Actor|SetLifeSpan (Math|Random|RandomFloatinRange 0.2 2.0)))

(fn GetChanceToSpawn
  (return (<= (Math|Random|RandomFloatinRange 0.0 1.0) (ScalableFloat|GetValueatLevel (Variables|Default|GetHF_SpawnChance)))))

(fn ConstructionScript)
